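# Looping sound cue for the diegetic music of the FishStick quest at a point of interest, filed under the FancyDiner quest sounds.
# A single wave player loops one music wave, shaped by a quest-specific attenuation asset with a short maximum distance.
# It routes through the third-person emote music sound class and keeps playing when silent.
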SoundCue /Game/Athena/Sounds/Quests/FancyDiner/SC_POI_FishStickQuest_DiegeticMus_Cue
root=WavePlayer_0
nodes [{id=WavePlayer_0, properties {bLooping=true, soundWaveAssetPtr=&/POI_FishStickQuest_DiegeticMus}}]
properties {
 attenuationSettings=&/Fort_FishStickQuest_Att
 maxDistance=1600
 soundClassObject=/Game/Sounds/FortSoundClassesAndModes/Fort_Music_Emote_3P
 virtualizationMode=PlayWhenSilent
}
